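# Material function that adds a glowing "Slurp" liquid look to the Marine Tonic shoe, optionally blended with the incoming material attributes by an FX mask.
# Three panning, parallax-offset bubble layers on pre-skinned positions, or a UV-scrolled fallback, feed inner/outer color blends with fake sun specular and scattering.
# Shading evaluates spherical harmonics decoded from vertex color against a distorted, skinning-rotatable direction.
# An optional Bubbly Mode shifts colors, speed and distortion; animated noise drives world position offset.
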MaterialFunction /CosmeticShoes/Assets/FORT_Shoes/Shoe_MarineTonic/Materials/MF_Slurp_MarineTonic
graph {nodes [
  n1 = NamedRerouteUsage()
  n2 = NamedRerouteUsage()
  n3 = NamedRerouteUsage()
  FunctionOutput(None=StaticSwitch(
    True=BlendMaterialAttributes(A=@n1, B=@n2, Alpha=@n3)
    False=NamedRerouteUsage()
    Value=StaticBoolParameter() {group=Slurp, parameterName=Use FXmask for Slurp}
   ))
  n4 = ScalarParameter() {defaultValue=3, parameterName=BubblesQuantSteps}
  n5 = Multiply(A=Saturate(None=ScalarParameter() {group=Slurp_BubblyMode, parameterName=BubblyMode_intensity}), B=@n4)
  NamedRerouteDeclaration(None=Divide(A=Floor(None=@n5), B=@n4)) {name=BubblyModeIntensityQuant, nodeColor=<rgba 0 1 0.5313 1>}
  n6 = ScalarParameter() {group=Slurp, parameterName=Bubble Speed X_1}
  n7 = ScalarParameter() {defaultValue=0.025, group=Slurp, parameterName=Bubble Speed Y_1}
  n8 = ScalarParameter() {defaultValue=1, group=Slurp, parameterName=BubbleSpeedGlobal}
  n9 = StaticSwitch(
   True=LinearInterpolate(A=@n8, B=ScalarParameter() {defaultValue=46, group=Slurp_BubblyMode, parameterName=BubbleSpeedGlobal_Bubbly}, Alpha=NamedRerouteUsage())
   False=@n8
   Value=StaticBoolParameter() {group=Slurp_BubblyMode, parameterName=EnableBubblyMode}
  )
  NamedRerouteDeclaration(None=Multiply(A=AppendVector(A=@n6, B=@n7), B=@n9)) {name=BubbleSpeed_1, nodeColor=<rgba 1 0 0.5625 1>}
  n10 = NamedRerouteUsage()
  n11 = MaterialFunctionCall(Float3=@n10) {materialFunction=&BreakOutFloat3Components}
  n12 = MaterialFunctionCall(X=@n11.R, Y=@n11.R, Z=@n11.R) {materialFunction=&MakeFloat3}
  n13 = Multiply(A=NamedRerouteUsage(), B=ScalarParameter() {group=Slurp, parameterName=UseTopDownMappingOnSoles})
  n14 = NamedRerouteUsage()
  n15 = LinearInterpolate(
   A=StaticSwitchParameter(True=@n12.Result, False=LinearInterpolate(A=@n10, B=@n12.Result, Alpha=@n13)) {group=Slurp, parameterName=Bubble Top Down Mapping}
   B=@n12.Result
   Alpha=@n14
  ) {constAlpha=1}
  n16 = CameraPositionWS()
  n17 = MaterialFunctionCall() {materialFunction=&ComponentPivotLocation}
  n18 = StaticSwitchParameter(
   True=MaterialFunctionCall(
    VectorToTransform=Subtract(A=@n16, B=@n17.Result)
    BasisX=VectorParameter() {parameterName=ActorRotationX, defaultValue=<rgba 1 0 0 1>}.RGB
    BasisY=VectorParameter() {parameterName=ActorRotationY, defaultValue=<rgba 0 1 0 1>}.RGB
    BasisZ=VectorParameter() {parameterName=ActorRotationZ, defaultValue=<rgba 0 0 1 1>}.RGB
   ) {materialFunction=&Transform3x3Matrix}.Transformed Vector
   False=TransformPosition(None=@n16) {transformSourceType=TRANSFORMPOSSOURCE_World}
  ) {group=Slurp, parameterName=UseBPRotationMatrix}
  n19 = MaterialFunctionCall(Float3=@n18) {materialFunction=&BreakOutFloat3Components}
  n20 = MaterialFunctionCall(X=@n19.R, Y=@n19.R, Z=@n19.R) {materialFunction=&MakeFloat3}
  n21 = NamedRerouteUsage()
  n22 = ScalarParameter() {group=Slurp, parameterName=UseTopDownMappingOnSoles}
  n23 = StaticSwitchParameter(True=@n20.Result, False=LinearInterpolate(A=@n20.Result, B=@n18, Alpha=Multiply(A=@n21, B=@n22))) {group=Slurp, parameterName=Bubble Top Down Mapping}
  n24 = Normalize(VectorInput=Subtract(A=@n15, B=LinearInterpolate(A=@n23, B=@n20.Result, Alpha=@n14)))
  n25 = Multiply(A=@n24, B=ScalarParameter() {defaultValue=15, group=Slurp, parameterName=Bubble_ParallaxDepth_Layer3}) {constB=15}
  n26 = ComponentMask(None=@n15) {g=true, r=true}
  n27 = Saturate(None=Sign(None=MaterialFunctionCall(XY=@n26) {materialFunction=&Swizzle}.YX))
  n28 = ScalarParameter() {group=Slurp, parameterName=TriPlanarSymmetryOffset}
  n29 = AppendVector(A=Add(A=ComponentMask(None=@n15) {g=true, r=true}, B=Multiply(A=@n27, B=@n28)), B=ComponentMask(None=@n15) {b=true})
  n30 = Add(A=@n29) {constB=166}
  n31 = QualitySwitch(Default=Add(A=Add(A=@n29, B=@n25)) {constB=166}, Low=@n30)
  n32 = Add(
   A=ShadingPathSwitch(Default=@n31, Deferred=@n31, Forward=@n30, Mobile=@n30)
   B=VectorParameter() {group=Slurp, parameterName=Bubble Position_3_Offset, defaultValue=<rgba 0 0 0 1>}.RGB
  )
  n33 = MaterialFunctionCall(Float3=@n32) {materialFunction=&BreakOutFloat3Components}
  NamedRerouteDeclaration(None=StaticSwitch(
    True=MaterialFunctionCall(X=@n33.R, Y=@n33.R, Z=@n33.R) {materialFunction=&MakeFloat3}.Result
    False=@n32
    Value=StaticBoolParameter() {group=Slurp, parameterName=Bubble_FrontPlanarProjectionOnly}
   )) {name=BubbleWorldPosParallax_3, nodeColor=<rgba 1 0 0.0469 1>}
  n34 = PreSkinnedPosition()
  n35 = VertexInterpolator(VS=@n34)
  NamedRerouteDeclaration(None=FunctionInput(Preview=ShaderStageSwitch(PixelShader=@n35.PS, VertexShader=@n34)) {bUsePreviewValueAsDefault=true, inputName=Pre Skinned Position}) {name=PreSkinnedPos_VS-PS, nodeColor=<rgba 0 0.4844 1 1>}
  n36 = Multiply(A=@n24, B=ScalarParameter() {defaultValue=15, group=Slurp, parameterName=Bubble_ParallaxDepth_Layer2}) {constB=15}
  n37 = Add(A=@n29) {constB=20}
  n38 = QualitySwitch(Default=Add(A=Add(A=@n29, B=@n36)) {constB=20}, Low=@n37)
  n39 = Add(
   A=ShadingPathSwitch(Default=@n38, Deferred=@n38, Forward=@n37, Mobile=@n37)
   B=VectorParameter() {group=Slurp, parameterName=Bubble Position_2_Offset, defaultValue=<rgba 0 0 0 1>}.RGB
  )
  NamedRerouteDeclaration(None=@n39) {name=BubbleWorldPosParallax_2_preSwizzle, nodeColor=<rgba 0.0781 0 1 1>}
  NamedRerouteDeclaration(None=StaticBoolParameter() {group=Slurp, parameterName=UseBubbly_Colors_Outer}) {name=BubblyColorsOuter_State, nodeColor=<rgba 1 0.7734 0 1>}
  NamedRerouteDeclaration(None=StaticBoolParameter() {group=Slurp, parameterName=UseBubbly_Colors_Inner}) {name=BubblyColorsInner_State, nodeColor=<rgba 1 0.7734 0 1>}
  NamedRerouteDeclaration(None=StaticBoolParameter() {group=Slurp_BubblyMode, parameterName=EnableBubblyMode}) {name=Bubbly_State, nodeColor=<rgba 0.0193 0.7995 1 1>}
  n40 = ScalarParameter() {defaultValue=12, group=Slurp, parameterName=Bubble_ParallaxDepth_Layer1}
  n41 = QualitySwitch(Default=Add(A=@n29, B=Multiply(A=@n24, B=@n40) {constB=12}), Low=@n29)
  n42 = Add(
   A=ShadingPathSwitch(Default=@n41, Deferred=@n41, Forward=@n29, Mobile=@n29)
   B=VectorParameter() {group=Slurp, parameterName=Bubble Position_1_Offset, defaultValue=<rgba 0 0 0 1>}.RGB
  )
  n43 = MaterialFunctionCall(Float3=@n42) {materialFunction=&BreakOutFloat3Components}
  NamedRerouteDeclaration(None=StaticSwitch(
    True=MaterialFunctionCall(X=@n43.R, Y=@n43.R, Z=@n43.R) {materialFunction=&MakeFloat3}.Result
    False=@n42
    Value=StaticBoolParameter() {group=Slurp, parameterName=Bubble_FrontPlanarProjectionOnly}
   )) {name=BubbleWorldPosParallax_1, nodeColor=<rgba 0.0781 0 1 1>}
  n44 = VectorParameter() {group=Slurp, parameterName=Bubble_DistortAmplitude_1, defaultValue=<rgba 0 0 0 1>}
  n45 = VectorParameter() {group=Slurp_BubblyMode, parameterName=Bubble_DistortAmplitude_1_Bubbly, defaultValue=<rgba 0.5 0.5 0.02 1>}
  n46 = NamedRerouteUsage()
  n47 = LinearInterpolate(A=@n44.RGB, B=@n45.RGB, Alpha=@n46)
  n48 = StaticBoolParameter() {group=Slurp_BubblyMode, parameterName=EnableBubblyMode}
  n49 = ComponentMask(None=NamedRerouteUsage()) {b=true}
  n50 = ScalarParameter() {defaultValue=0.1, group=Slurp, parameterName=Bubble_DistortTiling}
  n51 = StaticSwitch(
   True=LinearInterpolate(A=@n50, B=ScalarParameter() {defaultValue=0.01, group=Slurp_BubblyMode, parameterName=Bubble_DistortTiling_Bubbly}, Alpha=@n46)
   False=@n50
   Value=@n48
  )
  n52 = ScalarParameter() {defaultValue=0.15, group=Slurp, parameterName=Bubble_DistortSpeed}
  n53 = ScalarParameter() {defaultValue=0.89, group=Slurp_BubblyMode, parameterName=Bubble_DistortSpeed_Bubbly}
  n54 = Multiply(A=Time(), B=StaticSwitch(True=LinearInterpolate(A=@n52, B=@n53, Alpha=@n46), False=@n52, Value=@n48)) {constB=0.5}
  n55 = Sine(None=Add(A=Multiply(A=@n49, B=@n51), B=Frac(None=@n54)))
  NamedRerouteDeclaration(None=Multiply(A=StaticSwitch(True=@n47, False=@n44.RGB, Value=@n48), B=@n55)) {name=BubbleDistort_1, nodeColor=<rgba 1 0 0.3281 1>}
  n56 = VectorParameter() {group=Slurp, parameterName=Bubble_DistortAmplitude_2, defaultValue=<rgba 0 0 0 1>}
  n57 = LinearInterpolate(
   A=@n56.RGB
   B=VectorParameter() {group=Slurp_BubblyMode, parameterName=Bubble_DistortAmplitude_2_Bubbly, defaultValue=<rgba 1 1 0.02 1>}.RGB
   Alpha=@n46
  )
  NamedRerouteDeclaration(None=Multiply(A=StaticSwitch(True=@n57, False=@n56.RGB, Value=@n48), B=@n55)) {name=BubbleDistort_2, nodeColor=<rgba 1 0 0.3281 1>}
  n58 = ScalarParameter() {group=Slurp, parameterName=Bubble Speed X_2}
  n59 = ScalarParameter() {defaultValue=0.05, group=Slurp, parameterName=Bubble Speed Y_2}
  NamedRerouteDeclaration(None=Multiply(A=AppendVector(A=@n58, B=@n59), B=@n9)) {name=BubbleSpeed_2, nodeColor=<rgba 1 0 0.5625 1>}
  n60 = GetMaterialAttributes(None=NamedRerouteUsage()) {attributeGetTypes [
    #1
    #2
    #3
    #4
    #5
    #6
   ]}
  NamedRerouteDeclaration(None=@n60.WorldPositionOffset) {name=InWPO, nodeColor=<rgba 0 0.5312 1 1>}
  n61 = MaterialFunctionCall(Float3=@n39) {materialFunction=&BreakOutFloat3Components}
  NamedRerouteDeclaration(None=StaticSwitch(
    True=MaterialFunctionCall(X=@n61.R, Y=@n61.R, Z=@n61.R) {materialFunction=&MakeFloat3}.Result
    False=@n39
    Value=StaticBoolParameter() {group=Slurp, parameterName=Bubble_FrontPlanarProjectionOnly}
   )) {name=BubbleWorldPosParallax_2, nodeColor=<rgba 0.0781 0 1 1>}
  n62 = NamedRerouteUsage()
  n63 = MaterialFunctionCall(Float3=@n62) {materialFunction=&BreakOutFloat3Components}
  n64 = MaterialFunctionCall(X=@n63.R, Y=@n63.R, Z=@n63.R) {materialFunction=&MakeFloat3}
  n65 = LinearInterpolate(A=@n62, B=@n64.Result, Alpha=Multiply(A=NamedRerouteUsage(), B=ScalarParameter() {group=Slurp, parameterName=UseTopDownMappingOnSoles}))
  NamedRerouteDeclaration(None=LinearInterpolate(
    A=StaticSwitchParameter(True=@n64.Result, False=@n65) {group=Slurp, parameterName=Bubble Top Down Mapping}
    B=@n64.Result
    Alpha=NamedRerouteUsage()
   )) {name=PreskinnedNrm-Interp-SwizzledForTopDownMapping, nodeColor=<rgba 1 0 0.9609 1>}
  NamedRerouteDeclaration(None=StaticBoolParameter() {defaultValue=true, group=Slurp, parameterName=EnableBubbles}) {name=Enable_Bubbles_state, nodeColor=<rgba 1 0.8203 0 1>}
  NamedRerouteDeclaration(None=@n60.Specular) {name=InSpecular, nodeColor=<rgba 0.8516 1 0 1>}
  NamedRerouteDeclaration(None=@n60.Roughness) {name=InRoughness, nodeColor=<rgba 0.8281 0 1 1>}
  NamedRerouteDeclaration(None=@n60.Metallic) {name=InMetallic, nodeColor=<rgba 0 1 0.5078 1>}
  n66 = Constant() {r=1}
  n67 = FunctionInput() {bUsePreviewValueAsDefault=true, description=AmbientOcclusion, inputName=AO, inputType=FunctionInput_Scalar, previewValue=<1 1 1 1>}
  n68 = ScalarParameter() {group=Slurp, parameterName=Slurp_AO_Intensity}
  n69 = ScalarParameter() {defaultValue=0.66, group=Slurp, parameterName=Slurp_TimeofDayScale_Morning}
  n70 = ScalarParameter() {defaultValue=1, group=Slurp, parameterName=Slurp_TimeofDayScale_Day}
  n71 = ScalarParameter() {defaultValue=0.66, group=Slurp, parameterName=Slurp_TimeofDayScale_Evening}
  n72 = ScalarParameter() {defaultValue=0.5, group=Slurp, parameterName=Slurp_TimeofDayScale_Night}
  n73 = CollectionParameter() {
   collection=&FortniteMaterialParameters
   parameterName=SunlightDirectionalVector
  }
  n74 = ComponentMask(None=@n73) {b=true, g=true, r=true}
  n75 = NamedRerouteUsage()
  n76 = Multiply(A=Multiply(A=DotProduct(A=@n74, B=@n75)) {constB=2}, B=@n75)
  n77 = DotProduct(
   A=Normalize(VectorInput=Subtract(A=@n76, B=@n74))
   B=Multiply(A=MaterialFunctionCall() {materialFunction=&CameraDirectionVector}.Result)
  )
  n78 = Power(Base=Max(A=@n77) {constB=0}, Exp=ScalarParameter() {defaultValue=5, group=Slurp, parameterName=Bubble_BackSpecular_Power})
  n79 = ScalarParameter() {defaultValue=0.15, group=Slurp, parameterName=Bubble_BackSpecular_Scale}
  n80 = Multiply(
   A=VectorParameter() {group=Slurp, parameterName=Bubble_BackSpecular_Color, defaultValue=<rgba 1 1 1 1>}.RGB
   B=Saturate(None=Multiply(A=@n78, B=@n79) {constB=0.15})
  )
  n81 = Power(Base=Multiply(A=@n77) {constB=-1}, Exp=ScalarParameter() {defaultValue=50, group=Slurp, parameterName=Bubble_HighlightSpecular_Power})
  n82 = ScalarParameter() {defaultValue=0.55, group=Slurp, parameterName=Bubble_HighlightSpecular_Scale}
  n83 = Multiply(
   A=Saturate(None=Multiply(A=@n81, B=@n82) {constB=0.55})
   B=VectorParameter() {group=Slurp, parameterName=Bubble_HighlightSpecular_Color, defaultValue=<rgba 1 1 1 1>}.RGB
  )
  n84 = StaticBoolParameter() {defaultValue=true, group=Slurp, parameterName=UseTriPlanarBubbles}
  n85 = MaterialFunctionCall() {materialFunction=&Fresnel_Function}
  n86 = TextureObjectParameter() {
   group=Slurp
   parameterName=BubbleMask
   samplerType=SAMPLERTYPE_Masks
   texture=/Game/Characters/Player/Male/Large/Bodies/M_LRG_Slurp_Monster/Textures/T_SlurpBubble
  }
  n87 = MaterialFunctionCall!(
   TextureObject=@n86
   TextureSize=NamedRerouteUsage()
   WorldPosition=Add(A=NamedRerouteUsage(), B=NamedRerouteUsage())
   Export Float 4
   In=NamedRerouteUsage()
   World Space Normal=NamedRerouteUsage()
   ProjectionTransitionContrast=ScalarParameter() {defaultValue=2, group=Slurp, parameterName=Bubble_Blend_SharpnessMasks, sliderMax=1}
  ) {materialFunction=/Game/Athena/SafeZone/Preskinned_WP_Texture_ScalePan}
  n88 = StaticSwitch(True=@n87.Z Texture, False=@n87.Z Texture, Value=StaticBoolParameter() {group=Slurp, parameterName=Bubble_FrontPlanarProjectionOnly})
  n89 = LinearInterpolate(
   A=MaterialFunctionCall(Float3=@n88) {materialFunction=&BreakOutFloat3Components}.R
   B=MaterialFunctionCall(f3=@n88) {materialFunction=&AddComponents}.f3
   Alpha=NamedRerouteUsage()
  )
  n90 = NamedRerouteUsage()
  n91 = Abs(None=DotProduct(A=@n90, B=Constant3Vector() {constant=<rgba 1 0 0 0>}))
  n92 = Abs(None=DotProduct(A=@n90, B=Constant3Vector() {constant=<rgba 0 0 1 0>}))
  n93 = NamedRerouteUsage()
  n94 = MaterialFunctionCall!(
   WorldPosition=NamedRerouteUsage()
   World Space Normal=NamedRerouteUsage()
   UVSize_Front=@n93
   UVSize_Side=@n93
   UVSize_Top=@n93
   ProjectionTransitionContrast=Constant() {r=0.1}
  ) {materialFunction=/Game/Packages/Fortress_SharedMaterials/Base_Material_Functions/WorldAlignedTexture_UVsMasks}
  n95 = StaticSwitch(
   True=Power(Base=Max(A=@n91, B=@n92), Exp=ScalarParameter() {defaultValue=2, group=Slurp, parameterName=Bubble_Blend_FrontPlanarSharpness, sliderMax=1}) {constExponent=4}
   False=@n94.Mask_Top
   Value=StaticBoolParameter() {group=Slurp, parameterName=Bubble_FrontPlanarProjectionOnly}
  )
  n96 = Saturate(None=Subtract(A=@n89, B=@n95))
  n97 = TextureObjectParameter() {
   group=Slurp
   parameterName=BubbleMask
   samplerType=SAMPLERTYPE_Masks
   texture=&T_SlurpBubble
  }
  n98 = MaterialFunctionCall(
   TextureObject=@n97
   TextureSize=NamedRerouteUsage()
   WorldPosition=Add(A=NamedRerouteUsage(), B=NamedRerouteUsage())
   In=NamedRerouteUsage()
   World Space Normal=NamedRerouteUsage()
   ProjectionTransitionContrast=ScalarParameter() {defaultValue=2, group=Slurp, parameterName=Bubble_Blend_SharpnessMasks, sliderMax=1}
  ) {materialFunction=&Preskinned_WP_Texture_ScalePan}
  n99 = StaticSwitch(True=@n98.Z Texture, False=@n98.Z Texture, Value=StaticBoolParameter() {group=Slurp, parameterName=Bubble_FrontPlanarProjectionOnly})
  n100 = MaterialFunctionCall(Float3=@n99) {materialFunction=&BreakOutFloat3Components}
  n101 = MaterialFunctionCall(f3=@n99) {materialFunction=&AddComponents}
  n102 = NamedRerouteUsage()
  n103 = Saturate(None=Subtract(A=LinearInterpolate(A=@n100.G, B=@n101.f3, Alpha=@n102), B=@n95))
  n104 = Saturate(None=Multiply(A=OneMinus(None=@n85.Result), B=Add(A=@n96, B=@n103)))
  n105 = FunctionInput() {bUsePreviewValueAsDefault=true, description=FXMaskChannel, inputName=BubblesMask, inputType=FunctionInput_Scalar, previewValue=<1 1 1 1>}
  n106 = ScalarParameter() {defaultValue=1, group=Slurp, parameterName=SlurpBubblesMaskPower}
  n107 = StaticSwitchParameter(True=TextureCoordinate() {coordinateIndex=1}, False=TextureCoordinate()) {group=Slurp, parameterName=UseTexCoord1ForDistortionAndBubbles}
  n108 = ComponentMask(None=NamedRerouteUsage()) {r=true}
  n109 = Multiply(A=Saturate(None=Sign(None=@n108)), B=ScalarParameter() {group=Slurp, parameterName=SlurpNormalSymmetryOffset})
  n110 = ComponentMask(None=@n107) {g=true}
  n111 = Panner(
   Coordinate=AppendVector(A=ComponentMask(None=@n107) {r=true}, B=Add(A=@n109, B=@n110))
   Speed=VectorParameter() {group=Slurp, parameterName=SlurpNormalScrollingSpeed, defaultValue=<rgba 0 -5 0 1>}.RGB
  )
  n112 = ComponentMask(None=VectorParameter() {group=Slurp, parameterName=SlurpNormalTiling, defaultValue=<rgba 1 1 0 0>}.RGB) {g=true, r=true}
  n113 = TextureSampleParameter2D(UVs=VertexInterpolator(VS=Multiply(A=@n111, B=@n112)).PS) {group=Slurp, parameterName=SlurpNormal, samplerType=SAMPLERTYPE_Normal, texture=/Game/Global/Textures/Default/Noise/T_SwirlNoise_N}
  n114 = Multiply(A=OneMinus(None=Power(Base=@n113.B, Exp=@n106)), B=ScalarParameter() {defaultValue=1, group=Slurp, parameterName=SlurpBubblesMaskMultipler})
  n115 = FunctionInput() {
   bUsePreviewValueAsDefault=true
   description=FXMaskChannel
   inputName=SlurpNormalDistortionMask
   inputType=FunctionInput_Scalar
   previewValue=<1 1 1 1>
  }
  n116 = Multiply(
   A=ScalarParameter() {defaultValue=1, group=Slurp, parameterName=SlurpNormalDistortionIntensity}
   B=ComponentMask(None=@n113.RGB) {g=true, r=true}
  )
  n117 = Normalize(VectorInput=AppendVector(A=Multiply(A=@n115, B=@n116), B=ComponentMask(None=@n113.RGB) {b=true}))
  n118 = ComponentMask(None=@n117) {g=true, r=true}
  n119 = ComponentMask(None=VectorParameter() {group=Slurp, parameterName=SlurpBubblesNormalDistortion}.RGB) {g=true, r=true}
  n120 = Panner(
   Coordinate=Add(A=Multiply(A=@n118, B=@n119), B=@n107)
   Speed=VectorParameter() {group=Slurp, parameterName=SlurpBubblesScrollingSpeed, defaultValue=<rgba 0 -5 0 1>}.RGB
  )
  n121 = Multiply(
   A=ComponentMask(None=VectorParameter() {group=Slurp, parameterName=SlurpBubblesTilingR, defaultValue=<rgba 1 1 0 0>}.RGB) {g=true, r=true}
   B=@n120
  )
  n122 = TextureObjectParameter() {
   group=Slurp
   parameterName=BubbleMask
   samplerType=SAMPLERTYPE_Masks
   texture=&T_SlurpBubble
  }
  n123 = TextureSample(UVs=@n121, Tex=@n122)
  n124 = Multiply(
   A=@n120
   B=ComponentMask(None=VectorParameter() {group=Slurp, parameterName=SlurpBubblesTilingG, defaultValue=<rgba 1 1 0 0>}.RGB) {g=true, r=true}
  )
  n125 = TextureSample(UVs=@n124, Tex=@n122)
  n126 = Add(A=Add(A=@n123.R, B=@n125.G)) {constB=0, desc=0}
  n127 = OneMinus(None=Fresnel() {baseReflectFraction=0, exponent=1})
  n128 = Multiply(A=@n105, B=Multiply(A=Saturate(None=@n114), B=Multiply(A=@n126, B=@n127)))
  n129 = CollectionParameter() {
   collection=&FortniteMaterialParameters
   parameterName=SunlightDirectionalVector
  }
  n130 = ComponentMask(None=@n129) {b=true, g=true, r=true}
  n131 = AppendVector(
   A=Multiply(A=ComponentMask(None=@n117) {g=true, r=true}, B=ScalarParameter() {group=Slurp, parameterName=SlurpNormalFakeSpecularDistortion})
   B=ComponentMask(None=@n117) {b=true}
  )
  n132 = StaticSwitch(
   True=Transform(None=Normalize(VectorInput=@n131))
   False=PixelNormalWS()
   Value=StaticBoolParameter() {group=Slurp, parameterName=UseSlurpNormalDistortion}
  )
  n133 = Multiply(A=Multiply(A=DotProduct(A=@n130, B=@n132)) {constB=2}, B=@n132)
  n134 = DotProduct(
   A=Normalize(VectorInput=Subtract(A=@n133, B=@n130))
   B=Multiply(A=MaterialFunctionCall() {materialFunction=&CameraDirectionVector}.Result) {constB=-1}
  )
  n135 = Add(A=Multiply(A=@n134) {constB=-1}, B=ScalarParameter() {group=Slurp, parameterName=Slurp_InnerBackScatteringBias})
  n136 = Multiply(
   A=Power(Base=Max(A=@n135) {constB=0}, Exp=ScalarParameter() {defaultValue=1, group=Slurp, parameterName=Slurp_InnerBackScatteringPower})
   B=ScalarParameter() {group=Slurp, parameterName=Slurp_InnerBackScatteringScale}
  ) {constB=0.55}
  n137 = VectorParameter() {group=Slurp, parameterName=SlurpEmissiveInnerColorMultiplier, defaultValue=<rgba 1 1 1 0>}
  n138 = NamedRerouteUsage()
  n139 = Multiply(A=@n138, B=@n137.RGB)
  n140 = TextureSampleParameter2D() {group=Slurp, parameterName=Slurp_CustomEmissiveColor, texture=&T_BlacksRGB}
  n141 = NamedRerouteUsage()
  n142 = LinearInterpolate(A=@n139, B=Multiply(A=@n140.RGB, B=@n137.RGB), Alpha=@n141)
  n143 = LinearInterpolate(
   A=VectorParameter() {group=Slurp, parameterName=Slurp_InnerColor, defaultValue=<rgba 0 0.0725 0.2031 1>}.RGB
   B=VectorParameter() {group=Slurp, parameterName=Slurp_SecondColor_InnerColor, defaultValue=<rgba 0 0.0725 0.2031 1>}.RGB
   Alpha=NamedRerouteUsage()
  )
  n144 = StaticSwitchParameter(True=StaticSwitchParameter(True=@n142, False=@n143) {group=Slurp, parameterName=SlurpUseCustomEmissiveAsInnerColor}, False=@n143) {group=Slurp, parameterName=SlurpUseEmissiveAsInnerColor}
  n145 = VectorParameter() {group=Slurp_BubblyMode, parameterName=Slurp_InnerColor_Bubbly, defaultValue=<rgba 0 0.0892 0.25 1>}
  n146 = Saturate(None=ScalarParameter() {group=Slurp_BubblyMode, parameterName=BubblyMode_intensity})
  n147 = NamedRerouteUsage()
  n148 = LinearInterpolate(
   A=VectorParameter() {group=Slurp, parameterName=Bubble_InnerColor, defaultValue=<rgba 0 0.0892 0.25 1>}.RGB
   B=VectorParameter() {group=Slurp, parameterName=Bubble_SecondColor_InnerColor, defaultValue=<rgba 0 0.0892 0.25 1>}.RGB
   Alpha=NamedRerouteUsage()
  )
  n149 = VectorParameter() {group=Slurp_BubblyMode, parameterName=Bubble_InnerColor_Bubbly, defaultValue=<rgba 0 0.0892 0.25 1>}
  n150 = LinearInterpolate(
   A=StaticSwitch(True=LinearInterpolate(A=@n144, B=@n145.RGB, Alpha=@n146), False=@n144, Value=@n147)
   B=StaticSwitch(True=LinearInterpolate(A=@n148, B=@n149.RGB, Alpha=@n146), False=@n148, Value=@n147)
   Alpha=NamedRerouteUsage()
  )
  n151 = StaticSwitch(True=LinearInterpolate(A=@n142, B=@n150, Alpha=NamedRerouteUsage()), False=@n142, Value=NamedRerouteUsage())
  n152 = NamedRerouteUsage()
  n153 = StaticBoolParameter() {defaultValue=true, group=Slurp, parameterName=UseTriPlanarBubbles}
  n154 = StaticSwitchParameter(
   True=StaticSwitchParameter(True=StaticSwitch(True=@n151, False=@n142, Value=@n152), False=@n139) {group=Slurp, parameterName=SlurpUseCustomEmissiveAsInnerColor}
   False=StaticSwitch(True=@n150, False=@n144, Value=@n153)
  ) {group=Slurp, parameterName=SlurpUseEmissiveAsInnerColor}
  n155 = ScalarParameter() {defaultValue=1, group=Slurp, parameterName=SlurpEmissiveOuterColorMultiplier}
  n156 = Multiply(A=@n138, B=@n155)
  n157 = LinearInterpolate(A=@n156, B=Multiply(A=@n140.RGB, B=@n155), Alpha=@n141)
  n158 = NamedRerouteUsage()
  n159 = LinearInterpolate(
   A=VectorParameter() {group=Slurp, parameterName=Slurp_OuterColorA, defaultValue=<rgba 0 0.8177 0.6624 1>}.RGB
   B=VectorParameter() {group=Slurp, parameterName=Slurp_SecondColor_OuterColorA, defaultValue=<rgba 0 0.8177 0.6624 1>}.RGB
   Alpha=@n158
  )
  n160 = StaticSwitchParameter(True=StaticSwitchParameter(True=@n157, False=@n159) {group=Slurp, parameterName=SlurpUseCustomEmissiveAsOuterColor}, False=@n159) {group=Slurp, parameterName=SlurpUseEmissiveAsOuterColor}
  n161 = VectorParameter() {group=Slurp_BubblyMode, parameterName=Slurp_OuterColorA_Bubbly, defaultValue=<rgba 0.75 2.5 2.3581 1>}
  n162 = Saturate(None=ScalarParameter() {group=Slurp_BubblyMode, parameterName=BubblyMode_intensity})
  n163 = NamedRerouteUsage()
  n164 = LinearInterpolate(
   A=VectorParameter() {group=Slurp, parameterName=Slurp_OuterColorB, defaultValue=<rgba 0 0.1672 1 0>}.RGB
   B=VectorParameter() {group=Slurp, parameterName=Slurp_SecondColor_OuterColorB, defaultValue=<rgba 0 0.1672 1 0>}.RGB
   Alpha=NamedRerouteUsage()
  )
  n165 = VectorParameter() {group=Slurp_BubblyMode, parameterName=Slurp_OuterColorB_Bubbly, defaultValue=<rgba 0 0.5903 0.75 0>}
  n166 = Subtract(
   A=ComponentMask(None=NamedRerouteUsage()) {b=true}
   B=ScalarParameter() {defaultValue=125, group=Slurp, parameterName=Slurp_OuterColor_TransitionStart}
  ) {constB=125}
  n167 = ScalarParameter() {defaultValue=0.015, group=Slurp, parameterName=Slurp_OuterColor_TransitionScale}
  n168 = LinearInterpolate(
   A=StaticSwitch(True=LinearInterpolate(A=@n160, B=@n161.RGB, Alpha=@n162), False=@n160, Value=@n163)
   B=StaticSwitch(True=LinearInterpolate(A=@n164, B=@n165.RGB, Alpha=@n162), False=@n164, Value=NamedRerouteUsage())
   Alpha=Saturate(None=Multiply(A=@n166, B=@n167) {constB=0.015})
  )
  n169 = LinearInterpolate(
   A=VectorParameter() {group=Slurp, parameterName=Bubble_OuterColor, defaultValue=<rgba 0 0.5903 0.75 0>}.RGB
   B=VectorParameter() {group=Slurp, parameterName=Bubble_SecondColor_OuterColor, defaultValue=<rgba 0 0.5903 0.75 0>}.RGB
   Alpha=@n158
  )
  n170 = VectorParameter() {group=Slurp_BubblyMode, parameterName=Bubble_OuterColor_Bubbly, defaultValue=<rgba 0.75 2.1273 2.5 0>}
  n171 = LinearInterpolate(A=@n168, B=StaticSwitch(True=LinearInterpolate(A=@n169, B=@n170.RGB, Alpha=@n162), False=@n169, Value=@n163), Alpha=NamedRerouteUsage())
  n172 = StaticSwitch(True=LinearInterpolate(A=@n157, B=@n171, Alpha=NamedRerouteUsage()), False=@n157, Value=NamedRerouteUsage())
  n173 = NamedRerouteUsage()
  n174 = StaticSwitchParameter(
   True=StaticSwitchParameter(True=StaticSwitch(True=@n172, False=@n157, Value=@n173), False=@n156) {group=Slurp, parameterName=SlurpUseCustomEmissiveAsOuterColor}
   False=StaticSwitch(True=@n171, False=@n168, Value=@n153)
  ) {group=Slurp, parameterName=SlurpUseEmissiveAsOuterColor}
  n175 = Multiply(A=@n94.UV_Front) {constB=0.25}
  n176 = Frac(None=Multiply(A=Time(), B=ScalarParameter() {defaultValue=0.05, group=Slurp, parameterName=Bubble_SHScaleNoiseSpeed}) {constB=0.05})
  n177 = AppendVector(A=ComponentMask(None=@n175) {r=true}, B=Add(A=ComponentMask(None=@n175) {g=true}, B=@n176))
  n178 = TextureObjectParameter() {
   group=Slurp
   parameterName=Bubble_SHScaleNoise
   samplerType=SAMPLERTYPE_Grayscale
   texture=&LowResBlurredNoise
  }
  n179 = Multiply(A=@n94.UV_Side) {constB=0.25}
  n180 = AppendVector(A=ComponentMask(None=@n179) {r=true}, B=Add(A=ComponentMask(None=@n179) {g=true}, B=@n176))
  n181 = Multiply(A=LinearInterpolate(
    A=TextureSample(UVs=@n177, Tex=@n178) {constMipValue=0, samplerType=SAMPLERTYPE_Grayscale, texture=&LowResBlurredNoise}.R
    B=TextureSample(UVs=@n180, Tex=@n178) {constMipValue=0, samplerType=SAMPLERTYPE_Grayscale, texture=&LowResBlurredNoise}.R
    Alpha=@n94.Mask_Side
   )) {constB=2}
  n182 = Saturate(None=@n181)
  n183 = Add(
   A=Multiply(A=@n96, B=ScalarParameter() {defaultValue=1, group=Slurp, parameterName=BubbleIntensity_A}) {constB=0.5}
   B=Multiply(A=@n103, B=ScalarParameter() {defaultValue=1, group=Slurp, parameterName=BubbleIntensity_A_1}) {constB=0.5}
  )
  n184 = TextureObjectParameter() {
   group=Slurp
   parameterName=BubbleMask
   samplerType=SAMPLERTYPE_Masks
   texture=&T_SlurpBubble
  }
  n185 = MaterialFunctionCall(
   TextureObject=@n184
   TextureSize=NamedRerouteUsage()
   WorldPosition=NamedRerouteUsage()
   In=NamedRerouteUsage()
   World Space Normal=NamedRerouteUsage()
   ProjectionTransitionContrast=ScalarParameter() {defaultValue=2, group=Slurp, parameterName=Bubble_Blend_SharpnessMasks, sliderMax=1}
  ) {materialFunction=&Preskinned_WP_Texture_ScalePan}
  n186 = StaticSwitch(True=@n185.Z Texture, False=@n185.Z Texture, Value=StaticBoolParameter() {group=Slurp, parameterName=Bubble_FrontPlanarProjectionOnly})
  n187 = MaterialFunctionCall(Float3=@n186) {materialFunction=&BreakOutFloat3Components}
  n188 = MaterialFunctionCall(f3=@n186) {materialFunction=&AddComponents}
  n189 = NamedRerouteUsage()
  n190 = Saturate(None=Subtract(A=LinearInterpolate(A=@n187.B, B=@n188.f3, Alpha=@n189), B=@n95))
  n191 = ScalarParameter() {defaultValue=0.5, group=Slurp, parameterName=BubbleIntensity_A_2}
  n192 = QualitySwitch(Default=Add(A=@n183, B=Multiply(A=@n190, B=@n191) {constB=0.5}), Low=@n183)
  n193 = ShadingPathSwitch(Default=@n192, Deferred=@n192, Forward=@n183, Mobile=@n183)
  n194 = Time()
  n195 = Panner(
   Coordinate=Multiply(
    A=MaterialFunctionCall!(ReflectionVector, ReflectionTexture) {materialFunction=/Engine/Functions/Engine_MaterialFunctions01/Reflections/ViewAlignedReflection}.UVs
    B=ScalarParameter() {defaultValue=1, group=Slurp, parameterName=Bubble_ViewDependantDistortion}
   )
   Time=@n194
   Speed=AppendVector(
    A=ScalarParameter() {defaultValue=0.025, group=Slurp, parameterName=Bubble_ViewDistortion_SpeedX}
    B=ScalarParameter() {defaultValue=0.05, group=Slurp, parameterName=Bubble_ViewDistortion_SpeedY}
   )
  )
  n196 = Multiply(
   A=TextureSample(UVs=Rotator(Coordinate=@n195, Time=@n194) {speed=0.0001}) {constMipValue=0, samplerType=SAMPLERTYPE_Grayscale, texture=&LowResBlurredNoise}.G
   B=Multiply(A=Power(Base=@n182) {constExponent=4}, B=NamedRerouteDeclaration(None=@n193) {name=Msk_Bubbles, nodeColor=<rgba 1 0 0.6328 1>})
  )
  n197 = OneMinus(None=Fresnel() {baseReflectFraction=0, exponent=1})
  n198 = VectorParameter() {group=Slurp, parameterName=SHMax, defaultValue=<rgba 1.25 0.13 0.18 0.13>}
  n199 = VectorParameter() {group=Slurp, parameterName=SHMin, defaultValue=<rgba 1 -0.16 -0.17 -0.15>}
  n200 = AppendVector(A=@n199.RGB, B=@n199.RGB)
  n201 = Subtract(A=AppendVector(A=@n198.RGB, B=@n198.RGB), B=@n200)
  n202 = VertexColor()
  n203 = AppendVector(A=@n202, B=@n202)
  n204 = Multiply(A=MaterialFunctionCall(X=@n202.G, Y=@n202.G, Z=@n202.G) {materialFunction=&MakeFloat3}.Result)
  MaterialExpressionStaticSwitchParameter_8 = StaticSwitchParameter(True=Add(A=Multiply(A=@n201, B=@n203), B=@n200), False=AppendVector(A=Add(A=@n202.R), B=Add(A=@n204) {constB=-0.5})) {defaultValue=true, group=Slurp, parameterName=UseCustomSH Fit}
  n205 = Transform(None=CameraVectorWS()) {transformSourceType=TRANSFORMSOURCE_World, transformType=TRANSFORM_Local}
  n206 = MaterialFunctionCall() {materialFunction=&CameraDirectionVector}
  n207 = ComponentMask(None=Multiply(
    A=StaticSwitchParameter(True=@n129, False=@n206.Result) {group=Slurp, parameterName=UseSunDirection}
    B=ScalarParameter() {defaultValue=1, group=Slurp, parameterName=Slurp_DirectionVectorFlip}
   )) {b=true, g=true, r=true}
  n208 = VectorParameter() {group=Slurp, parameterName=SlurpNormalSHVectorDistortion}
  n209 = Transform(None=Normalize(VectorInput=@n117))
  n210 = StaticSwitch(
   True=Add(A=Multiply(A=@n209, B=@n208.RGB), B=@n207)
   False=@n207
   Value=StaticBoolParameter() {group=Slurp, parameterName=UseSlurpNormalDistortion}
  )
  n211 = Multiply(A=@n183, B=@n181)
  n212 = StaticBoolParameter() {defaultValue=true, group=Slurp, parameterName=UseTriPlanarBubbles}
  n213 = ScalarParameter() {defaultValue=0.25, group=Slurp, parameterName=Bubble_SHScale}
  n214 = LinearInterpolate(
   A=@n213
   B=ScalarParameter() {defaultValue=0.5, group=Slurp_BubblyMode, parameterName=Bubble_SHScale_Bubbly}
   Alpha=Saturate(None=ScalarParameter() {group=Slurp_BubblyMode, parameterName=BubblyMode_intensity})
  )
  n215 = StaticBoolParameter() {group=Slurp_BubblyMode, parameterName=EnableBubblyMode}
  n216 = LinearInterpolate(
   A=@n210
   B=NamedRerouteUsage()
   Alpha=Multiply(A=StaticSwitch(True=@n211, False=@n128, Value=@n212), B=StaticSwitch(True=@n214, False=@n213, Value=@n215))
  )
  n217 = NamedRerouteUsage()
  n218 = Transform(None=StaticSwitch(True=Normalize(VectorInput=@n216), False=@n210, Value=@n217)) {transformSourceType=TRANSFORMSOURCE_World, transformType=TRANSFORM_Local}
  n219 = ScalarParameter() {defaultValue=1, group=Slurp, parameterName=CustomDirection_Bias}
  n220 = VectorParameter() {group=Slurp, parameterName=Slurp_CustomDirectionVector, defaultValue=<rgba 0 0 -1 1>}
  n221 = StaticSwitchParameter(
   True=Normalize(VectorInput=LinearInterpolate(A=@n205, B=@n220.RGB, Alpha=@n219))
   False=Normalize(VectorInput=LinearInterpolate(A=@n205, B=@n218, Alpha=@n219))
  ) {group=Slurp, parameterName=UseCustomDirection}
  n222 = VertexNormalWS()
  MaterialExpressionStaticSwitchParameter_1 = StaticSwitchParameter(
   True=MaterialFunctionCall!(
    VectorToRotate=@n221
    PreSkinnedLocalNormal=NamedRerouteUsage()
    VertexlNormalWS=Transform(None=@n222) {transformSourceType=TRANSFORMSOURCE_World, transformType=TRANSFORM_Local}
   ) {materialFunction=/Game/Packages/Fortress_SharedMaterials/Base_Material_Functions/RotateVector_FromSkinning}.RotatedVectorFromSkinning
   False=@n221
  ) {defaultValue=true, group=Slurp, parameterName=UseRotateVector_FromSkinning_SH_Dir}
  n223 = Custom!(SH=@MaterialExpressionStaticSwitchParameter_8, Direction=@MaterialExpressionStaticSwitchParameter_1) {
   code="//SH decode,  Direction = LightDir, SH = 2band coefficients. (encoded in Y-up)\r\nDirection = normalize(Direction);\r\nfloat4 shEvalDirection;\r\nshEvalDirection.x = 0.282095;\r\nshEvalDirection.y =0.488603 * Direction.z;\r\nshEvalDirection.z = 0.488603 * Direction.y;\r\nshEvalDirection.w =-0.488603 * Direction.x;\r\nreturn saturate(dot(SH, shEvalDirection));"
   description=SphericalHarmonicsEval
   outputType=CMOT_Float1
   inputs [
    {
     inputName=SH
     input {expression=&self:MaterialExpressionStaticSwitchParameter_8}
    }
    {
     inputName=Direction
     input {expression=&self:MaterialExpressionStaticSwitchParameter_1}
    }
   ]
  }
  n224 = MaterialFunctionCall(
   Input=@n223
   Input Low=ScalarParameter() {defaultValue=0.275, group=Slurp, parameterName=SHContrast_LowValue}
   Input High=ScalarParameter() {defaultValue=0.525, group=Slurp, parameterName=SHContrast_HighValue}
   Target Low=Constant()
   Target High=Constant() {r=1}
  ) {materialFunction=&RemapValueRange}
  n225 = Power(Base=Saturate(None=OneMinus(None=@n224.Result)), Exp=ScalarParameter() {defaultValue=6, group=Slurp, parameterName=SHPow})
  n226 = StaticSwitch(
   True=@n225
   False=StaticSwitch(True=@n225, False=VertexInterpolator(VS=@n225).PS, Value=@n217)
   Value=StaticBoolParameter() {group=Slurp, parameterName=UseSlurpNormalDistortion}
  )
  n227 = Add(
   A=@n226
   B=StaticSwitch(
    True=Multiply(A=@n196, B=@n197)
    False=Constant()
    Value=StaticBoolParameter() {defaultValue=true, group=Slurp, parameterName=UseTriPlanarBubbles}
   )
  ) {constB=0}
  n228 = MaterialFunctionCall() {materialFunction=/Engine/Functions/Engine_MaterialFunctions02/Texturing/3DSandMayaUVCoordinates}
  n229 = OneMinus(None=Saturate(None=ComponentMask(None=@n228.UVChannel1) {r=true}))
  n230 = ScalarParameter() {group=Slurp, parameterName=Slurp_UV1_Darkening}
  n231 = QualitySwitch(Default=Add(A=StaticSwitch(True=@n227, False=@n226, Value=@n217), B=Multiply(A=@n229, B=@n230) {constB=0.0333}), Low=@n226)
  n232 = TextureSampleParameter2D() {group=Slurp, parameterName=Slurp_SHBiasMap, samplerType=SAMPLERTYPE_Masks, texture=/Game/Global/Textures/Default/Blanks/T_MidgreyMask}
  n233 = Subtract(A=@n232.G) {constB=0.5}
  n234 = Add(
   A=ShadingPathSwitch(Default=@n231, Deferred=@n231, Forward=@n226, Mobile=@n226)
   B=Multiply(A=Multiply(A=@n233) {constB=2}, B=ScalarParameter() {defaultValue=1, group=Slurp, parameterName=Slurp_SHBias_ScaleFactor})
  )
  n235 = ComponentMask(None=NamedRerouteUsage()) {b=true}
  n236 = OneMinus(None=Saturate(None=Divide(A=@n235) {constB=1.7}))
  n237 = Saturate(None=Add(A=Saturate(None=@n234), B=Multiply(A=@n236) {constB=0.14}))
  n238 = VectorParameter() {group=Slurp, parameterName=SlurpEmissiveInnerScatteringColorMultiplier, defaultValue=<rgba 1 1 1 0>}
  n239 = Multiply(A=@n138, B=@n238.RGB)
  n240 = StaticSwitchParameter(True=LinearInterpolate(A=@n239, B=Multiply(A=@n140.RGB, B=@n238.RGB), Alpha=@n141), False=@n239) {group=Slurp, parameterName=SlurpUseCustomEmissiveAsInnerScatteringColor}
  n241 = LinearInterpolate(
   A=VectorParameter() {group=Slurp, parameterName=Slurp_InnerScatteringColor, defaultValue=<rgba 0.0841 0 0.2 1>}.RGB
   B=VectorParameter() {group=Slurp, parameterName=Slurp_SecondColor_InnerScatteringColor, defaultValue=<rgba 0.0841 0 0.2 1>}.RGB
   Alpha=NamedRerouteUsage()
  )
  n242 = Multiply(
   A=Power(Base=Max(A=@n134) {constB=0}, Exp=ScalarParameter() {defaultValue=5, group=Slurp, parameterName=Slurp_InnerScatteringPower})
   B=ScalarParameter() {defaultValue=0.4, group=Slurp, parameterName=Slurp_InnerScatteringScale}
  ) {constB=0.4}
  n243 = Multiply(
   A=LinearInterpolate(
    A=Constant3Vector() {constant=<rgba 1 1 1 0>}
    B=VectorParameter() {group=Slurp, parameterName=Slurp_InnerBackScatteringMultiplier, defaultValue=<rgba 0 0 0 1>}.RGB
    Alpha=Saturate(None=@n136)
   )
   B=LinearInterpolate(
    A=LinearInterpolate(A=@n154, B=@n174, Alpha=@n237)
    B=StaticSwitchParameter(True=@n240, False=@n241) {group=Slurp, parameterName=SlurpUseEmissiveAsInnerScatteringColor}
    Alpha=Saturate(None=@n242)
   )
  )
  n244 = Add(A=@n243, B=Multiply(A=Add(A=@n80, B=@n83), B=StaticSwitch(True=@n104, False=@n128, Value=@n84)))
  n245 = Multiply(
   A=Power(Base=LinearInterpolate(A=@n66, B=@n67, Alpha=@n68), Exp=ScalarParameter() {defaultValue=1, group=Slurp, parameterName=Slurp_AO_pow})
   B=Multiply(
    A=MaterialFunctionCall(Value_Morning=@n69, Value_Day=@n70, Value_Evening=@n71, Value_Night=@n72) {materialFunction=&Scalar_Time_of_Day}.Result
    B=StaticSwitch(True=@n244, False=@n243, Value=@n217)
   )
  )
  n246 = Constant3Vector()
  n247 = Saturate(None=LinearInterpolate(
    A=ScalarParameter() {defaultValue=1, group=Slurp, parameterName=Slurp_BaseColor_Emissive_Distribution, sliderMax=1}
    B=ScalarParameter() {defaultValue=1, group=Slurp, parameterName=Slurp_SecondColor_Emissive_Distribution, sliderMax=1}
    Alpha=NamedRerouteUsage()
   ))
  NamedRerouteDeclaration(None=LinearInterpolate(A=@n245, B=@n246, Alpha=@n247)) {name=OutBaseColor, nodeColor=<rgba 0.8047 0 1 1>}
  NamedRerouteDeclaration(None=FunctionInput() {inputName=Material, inputType=FunctionInput_MaterialAttributes}) {name=InMaterial, nodeColor=<rgba 0 0.1328 1 1>}
  n248 = SetMaterialAttributes!(
   MaterialAttributes
   BaseColor=NamedRerouteUsage()
   EmissiveColor=NamedRerouteUsage()
   Roughness=NamedRerouteUsage()
   Specular=NamedRerouteUsage()
   Metallic=NamedRerouteUsage()
   Normal=NamedRerouteUsage()
   WorldPositionOffset=NamedRerouteUsage()
  ) {attributeSetTypes [
    #7
    #2
    #4
    #3
    #5
    #1
    #6
   ]}
  NamedRerouteDeclaration(None=@n248) {name=OutMaterial, nodeColor=<rgba 0.4531 1 0 1>}
  n249 = ScalarParameter() {group=Slurp, parameterName=Bubble Speed X_3}
  n250 = ScalarParameter() {defaultValue=0.12, group=Slurp, parameterName=Bubble Speed Y_3}
  NamedRerouteDeclaration(None=Multiply(A=AppendVector(A=@n249, B=@n250), B=@n9)) {name=BubbleSpeed_3, nodeColor=<rgba 1 0 0.5625 1>}
  n251 = FunctionInput() {bUsePreviewValueAsDefault=true, description=FXMaskChannel, inputName=SlurpMask, inputType=FunctionInput_Scalar, previewValue=<1 1 1 1>}
  NamedRerouteDeclaration(None=@n251) {name=InSlurpMask, nodeColor=<rgba 1 0 0.7734 1>}
  n252 = Multiply(A=NamedRerouteUsage(), B=ScalarParameter() {defaultValue=1, parameterName=Pre_Slurp_WPO_Scale})
  n253 = VertexNormalWS()
  n254 = WorldPosition() {worldPositionShaderOffset=WPT_ExcludeAllShaderOffsets}
  n255 = Subtract(A=Frac(None=Multiply(A=@n254.XYZ) {constB=0.0025})) {constB=0.5}
  n256 = MaterialFunctionCall(Float3=Abs(None=Multiply(A=@n255) {constB=2})) {materialFunction=&BreakOutFloat3Components}
  n257 = Subtract(A=Add(A=Add(A=@n256.R, B=@n256.R), B=@n256.B)) {constB=2}
  n258 = Frac(None=Multiply(A=Time()) {constB=0.25})
  n259 = Add(A=Sine(None=Subtract(A=@n257, B=@n258)))
  n260 = MaterialFunctionCall(Float3=Multiply(A=NamedRerouteUsage(), B=ScalarParameter() {defaultValue=0.02, group=Slurp, parameterName=Slurp_WPO_LocalNoise_Size}) {constB=0.02}) {materialFunction=&BreakOutFloat3Components}
  n261 = Add(A=Add(A=@n260.R, B=@n260.R), B=@n260.B)
  n262 = Multiply(
   A=Subtract(A=Time(), B=ScalarParameter() {group=Slurp, parameterName=Slurp_WPO_random_seed})
   B=ScalarParameter() {defaultValue=0.6, group=Slurp, parameterName=Slurp_WPO_LocalNoise_Speed}
  ) {constB=0.6}
  n263 = Sine(None=Subtract(A=Subtract(A=@n261) {constB=2}, B=Frac(None=@n262)))
  n264 = Add(
   A=Multiply(A=Multiply(A=@n259) {constB=0.5}, B=ScalarParameter() {defaultValue=1, group=Slurp, parameterName=Slurp_WPO_WorldNoiseScale})
   B=Multiply(A=Add(A=@n263)) {constB=0.5}
  )
  n265 = Multiply(
   A=Multiply(
    A=TextureSampleParameter2D() {parameterName=WPO_Mask, samplerType=SAMPLERTYPE_Masks, texture=&T_WhiteMask}.R
    B=Multiply(A=@n253, B=@n264)
   )
   B=ScalarParameter() {defaultValue=0.75, group=Slurp, parameterName=Slurp_WPO_FinalScale}
  )
  NamedRerouteDeclaration(None=StaticSwitchParameter(True=Add(A=@n252, B=@n265), False=NamedRerouteUsage()) {defaultValue=true, group=Slurp, parameterName=UseSlurpWPO}) {name=OutWPO, nodeColor=<rgba 0 1 0.9063 1>}
  NamedRerouteDeclaration(None=LinearInterpolate(A=@n246, B=@n245, Alpha=@n247)) {name=OutEmissiveColor, nodeColor=<rgba 0 1 0.4844 1>}
  n266 = NamedRerouteUsage()
  n267 = NamedRerouteUsage()
  NamedRerouteDeclaration(None=StaticSwitch(
    True=MaterialFunctionCall(BaseNormal=@n266, AdditionalNormal=@n267) {materialFunction=&BlendAngleCorrectedNormals}.Result
    False=NamedRerouteUsage()
    Value=StaticBoolParameter() {group=Slurp, parameterName=UseSlurpNormalDistortion}
   )) {name=OutNormal, nodeColor=<rgba 1 0.1641 0 1>}
  NamedRerouteDeclaration(None=LinearInterpolate(
    A=ScalarParameter() {group=Slurp, parameterName=Slurp_Metalness}
    B=NamedRerouteUsage()
    Alpha=ScalarParameter() {group=Slurp, parameterName=Slurp_MetalnessMap_Intensity}
   )) {name=OutMetallic, nodeColor=<rgba 0 0.1562 1 1>}
  NamedRerouteDeclaration(None=LinearInterpolate(
    A=ScalarParameter() {defaultValue=0.2, group=Slurp, parameterName=Slurp_Roughness}
    B=NamedRerouteUsage()
    Alpha=ScalarParameter() {group=Slurp, parameterName=Slurp_RoughnessMap_Intensity}
   )) {name=OutRoughness, nodeColor=<rgba 0.4766 1 0 1>}
  NamedRerouteDeclaration(None=LinearInterpolate(
    A=ScalarParameter() {defaultValue=0.5, group=Slurp, parameterName=Slurp_Specular}
    B=NamedRerouteUsage()
    Alpha=ScalarParameter() {group=Slurp, parameterName=Slurp_SpecularMap_Intensity}
   )) {name=OutSpecular, nodeColor=<rgba 1 0 0.7969 1>}
  n268 = Constant3Vector()
  n269 = NamedRerouteUsage()
  n270 = Divide(A=Add(A=NamedRerouteUsage(), B=NamedRerouteUsage()), B=Multiply(A=Abs(None=@n269)) {constB=-1})
  n271 = Time()
  n272 = NamedRerouteUsage()
  n273 = TextureObjectParameter() {
   group=Slurp
   parameterName=BubbleNormal
   samplerType=SAMPLERTYPE_Normal
   texture=/Game/Characters/Player/Male/Large/Bodies/M_LRG_Slurp_Monster/Textures/T_SlurpBubble_N
  }
  n274 = TextureSample(UVs=Panner(Coordinate=ComponentMask(None=@n270) {b=true, r=true}, Time=@n271, Speed=@n272), Tex=@n273) {samplerType=SAMPLERTYPE_Normal}
  n275 = MaterialFunctionCall(X=@n274.R, Z=@n274.R) {materialFunction=&MakeFloat3}
  n276 = NamedRerouteUsage()
  n277 = ScalarParameter() {defaultValue=0.5, group=Slurp, parameterName=Bubble_Blend_NormalSharpness, sliderMax=1}
  n278 = Max(A=Add(A=Abs(None=@n276), B=Multiply(A=@n277) {constB=-0.55}) {constB=-0.55}) {constB=0}
  n279 = Divide(A=@n278, B=DotProduct(A=@n278, B=Constant3Vector() {constant=<rgba 1 1 1 0>}))
  n280 = ComponentMask(None=@n279) {g=true}
  n281 = TextureSample(UVs=Panner(Coordinate=ComponentMask(None=@n270) {b=true, g=true}, Time=@n271, Speed=@n272), Tex=@n273) {samplerType=SAMPLERTYPE_Normal}
  n282 = MaterialFunctionCall(Y=@n281.R, Z=@n281.R) {materialFunction=&MakeFloat3}
  n283 = ComponentMask(None=@n279) {r=true}
  n284 = NamedRerouteUsage()
  n285 = Divide(A=Add(A=NamedRerouteUsage(), B=NamedRerouteUsage()), B=Multiply(A=Abs(None=@n284)) {constB=-1})
  n286 = NamedRerouteUsage()
  n287 = TextureSample(UVs=Panner(Coordinate=ComponentMask(None=@n285) {b=true, r=true}, Time=@n271, Speed=@n286), Tex=@n273) {samplerType=SAMPLERTYPE_Normal}
  n288 = MaterialFunctionCall(X=@n287.R, Z=@n287.R) {materialFunction=&MakeFloat3}
  n289 = ComponentMask(None=@n279) {g=true}
  n290 = TextureSample(UVs=Panner(Coordinate=ComponentMask(None=@n285) {b=true, g=true}, Time=@n271, Speed=@n286), Tex=@n273) {samplerType=SAMPLERTYPE_Normal}
  n291 = MaterialFunctionCall(Y=@n290.R, Z=@n290.R) {materialFunction=&MakeFloat3}
  n292 = ComponentMask(None=@n279) {r=true}
  n293 = LinearInterpolate(
   A=Add(A=Multiply(A=@n275.Result, B=@n280), B=Multiply(A=@n282.Result, B=@n283))
   B=Add(A=Multiply(A=@n288.Result, B=@n289), B=Multiply(A=@n291.Result, B=@n292))
   Alpha=Saturate(None=@n89)
  )
  n294 = Normalize(VectorInput=Add(A=NamedRerouteUsage(), B=LinearInterpolate(A=@n268, B=@n293, Alpha=@n104)))
  n295 = TextureObjectParameter() {
   group=Slurp
   parameterName=BubbleNormal
   samplerType=SAMPLERTYPE_Normal
   texture=&T_SlurpBubble_N
  }
  n296 = Add(A=Multiply(A=TextureSample(UVs=@n121, Tex=@n295).RGB, B=@n123.R), B=Multiply(A=TextureSample(UVs=@n124, Tex=@n295).RGB, B=@n125.G))
  NamedRerouteDeclaration(None=StaticSwitch(True=Transform(None=@n294) {transformSourceType=TRANSFORMSOURCE_Local}, False=Transform(None=@n296), Value=@n212)) {name=BubbleWSNormal, nodeColor=<rgba 1 0.9844 0 1>}
  NamedRerouteDeclaration(None=@n115) {name=InNormalDistortionMask, nodeColor=<rgba 0 0.4375 1 1>}
  NamedRerouteDeclaration(None=FunctionInput(Preview=Constant()) {bUsePreviewValueAsDefault=true, inputName=SecondColor_Mask, inputType=FunctionInput_Scalar}) {name=SecondColorMask, nodeColor=<rgba 1 0 0.375 1>}
  n297 = ScalarParameter() {defaultValue=1, group=Slurp, parameterName=BubblesSizeGlobal}
  n298 = StaticSwitch(
   True=LinearInterpolate(A=@n297, B=ScalarParameter() {defaultValue=0.33, group=Slurp_BubblyMode, parameterName=BubblesSizeGlobalBubbly}, Alpha=NamedRerouteUsage())
   False=@n297
   Value=StaticBoolParameter() {group=Slurp_BubblyMode, parameterName=EnableBubblyMode}
  )
  NamedRerouteDeclaration(None=Multiply(A=@n298, B=ScalarParameter() {defaultValue=25, group=Slurp, parameterName=BubbleSize_A})) {name=BubbleSize_A, nodeColor=<rgba 0.0312 0 1 1>}
  n299 = MaterialFunctionCall(Float3=@n60.Normal) {materialFunction=&BreakOutFloat3Components}
  n300 = Multiply(A=AppendVector(A=@n299.R, B=@n299.R), B=ScalarParameter() {defaultValue=1, group=Slurp, parameterName=BaseNormalIntensity})
  NamedRerouteDeclaration(None=Normalize(VectorInput=AppendVector(A=@n300, B=@n299.B))) {name=InNormal, nodeColor=<rgba 1 0.4688 0 1>}
  n301 = ComponentMask(None=NamedRerouteUsage()) {b=true}
  n302 = Divide(A=@n301) {constB=0.7}
  NamedRerouteDeclaration(None=OneMinus(None=Saturate(None=@n302))) {name=ShoeSolesMask_Soft, nodeColor=<rgba 0.6172 0 1 1>}
  n303 = ComponentMask(None=@n117) {g=true, r=true}
  n304 = ScalarParameter() {defaultValue=1, group=Slurp, parameterName=SlurpNormalMaterialSpecularDistortion}
  n305 = Normalize(VectorInput=AppendVector(A=Multiply(A=@n303, B=@n304), B=ComponentMask(None=@n117) {b=true}))
  NamedRerouteDeclaration(None=@n305) {name=OutNormalDistortion, nodeColor=<rgba 1 0 0.5859 1>}
  NamedRerouteDeclaration(None=@n60.EmissiveColor) {name=InEmissive, nodeColor=<rgba 0 0.9062 1 1>}
  NamedRerouteDeclaration(None=OneMinus(None=Step(X=@n301) {constY=2.5})) {name=ShoeSolesMask, nodeColor=<rgba 0.4062 0 1 1>}
  NamedRerouteDeclaration(None=Multiply(A=@n298, B=ScalarParameter() {defaultValue=30, group=Slurp, parameterName=BubbleSize_A_1})) {name=BubbleSize_A_1, nodeColor=<rgba 0.0312 0 1 1>}
  NamedRerouteDeclaration(None=Multiply(A=@n298, B=ScalarParameter() {defaultValue=15, group=Slurp, parameterName=BubbleSize_A_2})) {name=BubbleSize_A_2, nodeColor=<rgba 0.0312 0 1 1>}
  n306 = PreSkinnedNormal()
  NamedRerouteDeclaration(None=ShaderStageSwitch(PixelShader=VertexInterpolator(VS=@n306).PS, VertexShader=@n306)) {name=PreSkinnedNrm_VS-PS, nodeColor=<rgba 0 0.4844 1 1>}
  n307 = Multiply(
   A=ScalarParameter() {group=Slurp_BubblyMode, parameterName=BubblyMode_intensity}
   B=ScalarParameter() {defaultValue=0.5, parameterName=UnmaskAllBubbles}
  )
  NamedRerouteDeclaration(None=StaticSwitch(True=Saturate(None=@n307), False=Constant(), Value=StaticBoolParameter() {group=Slurp_BubblyMode, parameterName=EnableBubblyMode})) {name=BubblyModeIntensity, nodeColor=<rgba 0.7812 1 0 1>}
  NamedRerouteDeclaration(None=@n42) {name=BubbleWorldPosParallax_1_preSwizzle, nodeColor=<rgba 0.0781 0 1 1>}
  Reroute(None=@n305)
  n308 = ScalarParameter() {defaultValue=4, group=Slurp, parameterName=Decal_ParallaxDepth}
  n309 = QualitySwitch(Default=Add(A=@n15, B=Multiply(A=@n24, B=@n308) {constB=12}), Low=@n15)
  n310 = MaterialFunctionCall(Float3=Add(
    A=ShadingPathSwitch(Default=@n309, Deferred=@n309, Forward=@n15, Mobile=@n15)
    B=VectorParameter() {group=Slurp, parameterName=Decal Position_Offset, defaultValue=<rgba 0 0 0 1>}.RGB
   )) {materialFunction=&BreakOutFloat3Components}
  n311 = ScalarParameter() {defaultValue=35, group=Slurp, parameterName=DecalSize}
  n312 = MaterialFunctionCall(
   WorldPosition=MaterialFunctionCall(X=@n310.R, Y=@n310.R, Z=@n310.R) {materialFunction=&MakeFloat3}.Result
   World Space Normal=NamedRerouteUsage()
   UVSize_Front=@n311
   UVSize_Side=@n311
   UVSize_Top=@n311
   ProjectionTransitionContrast=Constant() {r=0.1}
  ) {materialFunction=&WorldAlignedTexture_UVsMasks}
  n313 = Saturate(None=Add(A=@n312.UV_Top))
  n314 = Ceil(None=Saturate(None=Subtract(A=@n312.UV_Top)))
  n315 = MaterialFunctionCall(Float2=Multiply(A=Floor(None=@n313), B=OneMinus(None=@n314))) {materialFunction=&BreakOutFloat2Components}
  n316 = VectorParameter() {group=Slurp, parameterName=Decal_UVFlip, defaultValue=<rgba 1 1 0 1>}
  n317 = TextureObjectParameter() {
   group=Slurp
   parameterName=DecalMask
   samplerType=SAMPLERTYPE_Masks
   texture=&T_SlurpBubble
  }
  n318 = TextureSample(UVs=Multiply(A=@n312.UV_Top, B=ComponentMask(None=@n316.RGB) {g=true, r=true}), Tex=@n317) {samplerType=SAMPLERTYPE_Masks, texture=&T_SlurpBubble}
  n319 = Subtract(
   A=Multiply(
    A=Multiply(A=@n315.R, B=@n315.R)
    B=StaticComponentMaskParameter(None=@n318.RGBA) {defaultA=true, group=Slurp, parameterName=DecalMaskChannel}
   )
   B=@n95
  )
  Reroute(None=Multiply(A=Saturate(None=@n319), B=ScalarParameter() {defaultValue=1, group=Slurp, parameterName=DecalIntensity}))
 ]}
