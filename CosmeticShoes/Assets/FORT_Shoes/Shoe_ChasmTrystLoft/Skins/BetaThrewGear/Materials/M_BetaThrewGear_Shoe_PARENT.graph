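# Parent material for the BetaThrewGear cosmetic shoe skin, adding a Slurp liquid effect to the shared 2023 character material.
# PreFX output passes through shoe reactivity into the MarineTonic Slurp function, masked by Tech Art Mask green and a second-color texture.
# Emissive is then rescaled by Fresnel and sky light versus camera direction, brightening at rims and when facing the sun.
# Sock, PostFX and shoe hologram functions finish the chain.
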Material /CosmeticShoes/Assets/FORT_Shoes/Shoe_ChasmTrystLoft/Skins/BetaThrewGear/Materials/M_BetaThrewGear_Shoe_PARENT
graph {
 properties {
  bUseMaterialAttributes=true
  bUsedWithClothing=true
  bUsedWithMeshDeformer=true
  bUsedWithMorphTargets=true
  bUsedWithNanite=true
  bUsedWithNiagaraMeshParticles=true
  bUsedWithSkeletalMesh=true
  pixelDepthOffsetMode=PDOM_Legacy
  shadingModel=MSM_FromMaterialExpression
  usedShadingModels=Default Lit | Subsurface | Clear Coat | Subsurface Profile | Cloth
 }
 nodes [
  n1 = MaterialFunctionCall() {materialFunction=&MF_CharacterParent_2023_PreFX}
  n2 = TextureSampleParameter2D() {
   group=Slurp
   parameterName=SecondColor_Mask
   samplerType=SAMPLERTYPE_Masks
   texture=/CosmeticShoes/Assets/FORT_Shoes/Shoe_ChasmTrystLoft/Skins/BetaThrewGear/Textures/T_BetaThrewGear_Shoe_FX
  }
  n3 = GetMaterialAttributes(None=MaterialFunctionCall!(
    SlurpMask=ComponentMask(None=@n1.Tech Art Mask) {g=true}
    AO
    BubblesMask
    SlurpNormalDistortionMask
    Pre Skinned Position=@n1.Pre skinned Local Position
    SecondColor_Mask=@n2.R
    Material=MaterialFunctionCall!(Material=@n1.Result) {materialFunction=/Game/Global/Materials/Cosmetics/Kicks/Shoes/Materials/MF_ShoeReactivityTest}.Result
   ) {materialFunction=/CosmeticShoes/Assets/FORT_Shoes/Shoe_MarineTonic/Materials/MF_Slurp_MarineTonic}.Result) {attributeGetTypes [#1]}
  n4 = SkyAtmosphereLightDirection()
  n5 = CameraVectorWS()
  n6 = Power(Base=ConstantBiasScale(None=DotProduct(A=@n4, B=@n5))) {constExponent=0.75}
  n7 = Multiply(
   A=LinearInterpolate(Alpha=Power(Base=@n1.Fresnel) {constExponent=3}) {constA=1, constB=2}
   B=LinearInterpolate(Alpha=OneMinus(None=@n6)) {constA=1, constB=2}
  )
  n8 = SetMaterialAttributes(MaterialAttributes=@n3.MaterialAttributes, Emissive Color=Multiply(A=Multiply(A=@n3.Emissive Color, B=@n7)) {constB=0.5}) {attributeSetTypes [#1]}
  MaterialFunctionCall!(In=MaterialFunctionCall(
    Material=MaterialFunctionCall!(InOpacity, Material=@n8, Mask=@n1.Tech Art Mask) {materialFunction=/Game/Global/Materials/Cosmetics/Kicks/Socks/Materials/MF_Sock}.Result
    M=@n1.M
    Pre Skinned Local Position=@n1.M
    Distance Blend=@n1.M
    Fresnel=@n1.M
   ) {materialFunction=&MF_CharacterParent_2023_PostFX}.Result) {materialFunction=/Game/Global/Materials/Cosmetics/Kicks/Shoes/Materials/MF_TechartShoeHologram}
 ]
}
